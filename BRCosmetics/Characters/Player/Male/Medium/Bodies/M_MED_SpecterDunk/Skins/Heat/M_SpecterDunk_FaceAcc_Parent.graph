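# Parent material for the SpecterDunk Heat skin's face accessory, overriding base color, roughness and emissive between the shared character PreFX and PostFX functions.
# A packed mask separates skull and hood.
# Hood glow blends a noise-modulated pre-skinned sphere mask gradient with atlas or Surge patterns.
# Skull glow layers Fresnel and pre-skinned normal-direction colors, while eyes add Fresnel plus procedural UV sphere-mask pupils and glows with tunable placement.
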Material /BRCosmetics/Characters/Player/Male/Medium/Bodies/M_MED_SpecterDunk/Skins/Heat/M_SpecterDunk_FaceAcc_Parent
graph {
 properties {
  bAutomaticallySetUsageInEditor=false
  bUseMaterialAttributes=true
  bUsedWithMeshDeformer=true
  bUsedWithMorphTargets=true
  bUsedWithSkeletalMesh=true
  pixelDepthOffsetMode=PDOM_Legacy
  shadingModel=MSM_FromMaterialExpression
  usedShadingModels=Default Lit | Subsurface | Clear Coat | Subsurface Profile | Cloth
 }
 nodes [
  n1 = NamedRerouteUsage()
  n2 = ScalarParameter() {group=SpecterDunkFaceAcc, parameterName=RoughnessOffset}
  n3 = SetMaterialAttributes!(
   MaterialAttributes=NamedRerouteUsage()
   Base Color=NamedRerouteUsage()
   Emissive Color=Add(A=NamedRerouteUsage(), B=NamedRerouteUsage())
   Metallic
   Specular
   Roughness=LinearInterpolate(A=@n1, B=Add(A=@n1, B=@n2), Alpha=NamedRerouteUsage())
  ) {attributeSetTypes [
    #1
    #2
    #3
    #4
    #5
   ]}
  n4 = MaterialFunctionCall() {materialFunction=&MF_CharacterParent_2023_PreFX}
  MaterialFunctionCall(Material=@n3, M=@n4.M, Pre Skinned Local Position=@n4.M, Distance Blend=@n4.M, Fresnel=@n4.M) {materialFunction=&MF_CharacterParent_2023_PostFX}
  n5 = MaterialFunctionCall() {materialFunction=/BRCosmetics/Characters/Player/Male/Medium/Bodies/M_MED_SpecterDunk/Materials/MF_ProjectNoise}
  n6 = LocalPosition() {localOrigin=InstancePreSkinning}
  n7 = Power(
   Base=SphereMask(
    A=VertexInterpolator(VS=@n6.XYZ).PS
    B=Constant4Vector() {constant=<rgba -0.8024 -11.2256 143.9227 1>}.RGB
    Radius=Constant() {r=24}
    Hardness=Constant()
   )
   Exp=Constant() {r=2}
  )
  n8 = LinearInterpolate(
   A=VectorParameter() {group=SpecterDunkFaceAcc, parameterName=HoodEmissiveColorA, defaultValue=<rgba 0.3 0.0337 0.0001 1>}.RGB
   B=VectorParameter() {group=SpecterDunkFaceAcc, parameterName=HoodEmissiveColorB, defaultValue=<rgba 3.5 0.8766 0 1>}.RGB
   Alpha=Add(A=Multiply(A=@n5.Result, B=@n7), B=@n7)
  )
  n9 = VectorParameter() {group=SpecterDunkFaceAcc, parameterName=HoodPatternColorA, defaultValue=<rgba 0.75 0.1417 0.06 1>}
  n10 = TextureSample(UVs=MaterialFunctionCall() {materialFunction=/BRCosmetics/Characters/Player/Male/Medium/Bodies/M_MED_SpecterDunk/Materials/MF_AtlasUV_2x2}.AtlasUV_2x2) {
   samplerType=SAMPLERTYPE_Masks
   texture=/BRCosmetics/Characters/Player/Male/Medium/Bodies/M_MED_SpecterDunk/Textures/T_SpecterDunk_Body_PatternA
  }
  n11 = TextureSample() {
   samplerType=SAMPLERTYPE_Masks
   texture=/BRCosmetics/Characters/Player/Male/Medium/Bodies/M_MED_SpecterDunk/Textures/T_SpecterDunk_Body_Surge_Pattern
  }
  n12 = LinearInterpolate(A=@n10.R, B=@n11.G, Alpha=ScalarParameter() {group=SpecterDunkFaceAcc, parameterName=Surge?})
  NamedRerouteDeclaration(None=Multiply(A=LinearInterpolate(A=@n8, B=@n9.RGB, Alpha=@n12), B=NamedRerouteUsage())) {name=HoodEmissive, nodeColor=<rgba 0 1 1 1>}
  n13 = GetMaterialAttributes(None=@n4.Result) {attributeGetTypes [
    #1
    #2
    #3
    #4
    #5
   ]}
  NamedRerouteDeclaration(None=@n13.Emissive Color) {name=In_EmissiveColor, nodeColor=<rgba 1 0.7734 0 1>}
  n14 = TextureSample() {
   samplerType=SAMPLERTYPE_Masks
   texture=/BRCosmetics/Characters/Player/Male/Medium/Bodies/M_MED_SpecterDunk/Textures/T_SpecterDunk_Hood_TAmask
  }
  NamedRerouteDeclaration(None=@n14.R) {name=SkullMask, nodeColor=<rgba 1 0.8672 0 1>}
  NamedRerouteDeclaration(None=@n14.G) {name=HoodMask, nodeColor=<rgba 0.5469 0 1 1>}
  NamedRerouteDeclaration(None=@n13.Roughness) {name=IN_Roughness, nodeColor=<rgba 1 0 0.8203 1>}
  n15 = Constant() {r=2}
  n16 = Saturate(None=Power(Base=Fresnel(ExponentIn=@n15), Exp=Constant() {r=3.273}))
  n17 = VectorParameter() {group=SpecterDunkFaceAcc, parameterName=EyeBall_FresnelColor, defaultValue=<rgba 9 3.3356 0 1>}
  n18 = VectorParameter() {group=SpecterDunkFaceAcc, parameterName=EyePupil_EyeColorB, defaultValue=<rgba 1.5 0.3659 0.0531 1>}
  n19 = VectorParameter() {group=SpecterDunkFaceAcc, parameterName=EyePupil_EyeColor, defaultValue=<rgba 18 6.6712 0 1>}
  n20 = NamedRerouteUsage()
  n21 = TextureSample() {samplerType=SAMPLERTYPE_Masks, texture=/BRCosmetics/Characters/Player/Male/Medium/Bodies/M_MED_SpecterDunk/Textures/T_SpecterDunk_Eye_Mask}
  n22 = Multiply(A=Add(A=Multiply(A=@n16, B=@n17.RGB), B=LinearInterpolate(A=@n18.RGB, B=@n19.RGB, Alpha=@n20)), B=@n21.G)
  n23 = SphereMask(A=TextureCoordinate(), B=Constant() {r=1}, Radius=Constant() {r=0.5})
  NamedRerouteDeclaration(None=Add(A=NamedRerouteUsage(), B=Multiply(A=@n22, B=@n23))) {name=EyeEmissive, nodeColor=<rgba 1 0.375 0 1>}
  n24 = VectorParameter() {group=SpecterDunkFaceAcc, parameterName=Pupil position, defaultValue=<rgba 0.86 0.935 0.857 0.82>}
  n25 = Power(
   Base=SphereMask(
    A=TextureCoordinate()
    B=AppendVector(A=@n24.R, B=@n24.R)
    Radius=ScalarParameter() {defaultValue=0.025, group=SpecterDunkFaceAcc, parameterName=Pupil_Radius_L}
    Hardness=ScalarParameter() {defaultValue=0.3, group=SpecterDunkFaceAcc, parameterName=Pupil_Hardness_L}
   )
   Exp=ScalarParameter() {defaultValue=5, group=SpecterDunkFaceAcc, parameterName=Pupil_Power_L}
  )
  n26 = TextureCoordinate()
  n27 = AppendVector(A=@n24.R, B=@n24.R)
  n28 = Constant() {r=0.009}
  n29 = Constant() {r=0.85}
  n30 = Multiply(A=Power(Base=SphereMask(A=@n26, B=@n27, Radius=@n28, Hardness=@n29), Exp=Constant() {r=2}), B=Constant() {r=10})
  n31 = Power(
   Base=SphereMask(
    A=TextureCoordinate()
    B=AppendVector(A=@n24.B, B=@n24.B)
    Radius=ScalarParameter() {defaultValue=0.025, group=SpecterDunkFaceAcc, parameterName=Pupil_Radius_R}
    Hardness=ScalarParameter() {defaultValue=0.3, group=SpecterDunkFaceAcc, parameterName=Pupil_Hardness_R}
   )
   Exp=ScalarParameter() {defaultValue=5, group=SpecterDunkFaceAcc, parameterName=Pupil_Power_R}
  )
  n32 = TextureCoordinate()
  n33 = AppendVector(A=@n24.B, B=@n24.B)
  n34 = Constant() {r=0.009}
  n35 = Constant() {r=0.85}
  n36 = Multiply(A=Power(Base=SphereMask(A=@n32, B=@n33, Radius=@n34, Hardness=@n35), Exp=Constant() {r=2}), B=Constant() {r=10})
  NamedRerouteDeclaration(None=Add(A=Add(A=@n25, B=@n30), B=Add(A=@n31, B=@n36))) {name=EyePupil, nodeColor=<rgba 1 0.8672 0 1>}
  n37 = VectorParameter() {group=SpecterDunkFaceAcc, parameterName=EyeGlowPosition, defaultValue=<rgba 0.176 0.155 0.332 0.16>}
  n38 = Power(
   Base=SphereMask(
    A=TextureCoordinate()
    B=AppendVector(A=@n37.R, B=@n37.R)
    Radius=ScalarParameter() {defaultValue=0.06, group=SpecterDunkFaceAcc, parameterName=EyeGlow_Radius_L}
    Hardness=ScalarParameter() {defaultValue=0.5, group=SpecterDunkFaceAcc, parameterName=EyeGlow_Hardness_L}
   )
   Exp=ScalarParameter() {defaultValue=8, group=SpecterDunkFaceAcc, parameterName=EyeGlow_Power_L}
  )
  n39 = Power(
   Base=SphereMask(
    A=TextureCoordinate()
    B=AppendVector(A=@n37.B, B=@n37.B)
    Radius=ScalarParameter() {defaultValue=0.06, group=SpecterDunkFaceAcc, parameterName=EyeGlow_Radius_R}
    Hardness=ScalarParameter() {defaultValue=0.5, group=SpecterDunkFaceAcc, parameterName=EyeGlow_Hardness_R}
   )
   Exp=ScalarParameter() {defaultValue=8, group=SpecterDunkFaceAcc, parameterName=EyeGlow_Power_R}
  )
  NamedRerouteDeclaration(None=LinearInterpolate(
    B=VectorParameter() {group=SpecterDunkFaceAcc, parameterName=EyeGlow, defaultValue=<rgba 0.503 0.0657 0.0503 1>}.RGB
    Alpha=Add(A=@n38, B=@n39)
   )) {name=EyeGlow, nodeColor=<rgba 0.5469 0 1 1>}
  NamedRerouteDeclaration(None=ComponentMask(None=@n4.M) {g=true, r=true}) {name=In_AO, nodeColor=<rgba 0 1 0.2266 1>}
  NamedRerouteDeclaration(None=@n13.Base Color) {name=IN_BaseColor, nodeColor=<rgba 0 1 0.2969 1>}
  NamedRerouteDeclaration(None=@n13.MaterialAttributes) {name=In_MA, nodeColor=<rgba 1 0 0.0234 1>}
  NamedRerouteDeclaration(None=LinearInterpolate(A=NamedRerouteUsage(), B=Constant(), Alpha=NamedRerouteUsage())) {name=Out_BaseColor, nodeColor=<rgba 0 0.3438 1 1>}
  n40 = NamedRerouteUsage()
  n41 = VectorParameter() {group=SpecterDunkFaceAcc, parameterName=SkullEmissive_ColorB, defaultValue=<rgba 2 0.5079 0.0515 1>}
  n42 = VectorParameter() {group=SpecterDunkFaceAcc, parameterName=SkullEmissive_ColorA, defaultValue=<rgba 1.5 0.13 0.105 1>}
  n43 = Constant() {r=0.03}
  n44 = Saturate(None=Power(Base=Fresnel(ExponentIn=@n43), Exp=Constant() {r=3.273}))
  n45 = TextureSample() {samplerType=SAMPLERTYPE_Masks, texture=&T_SpecterDunk_Eye_Mask}
  n46 = Constant4Vector() {constant=<rgba 0 0.2 1 1>}
  n47 = VertexInterpolator(VS=PreSkinnedNormal())
  n48 = Multiply(A=OneMinus(None=@n45.R), B=Saturate(None=DotProduct(A=@n46.RGB, B=@n47.PS)))
  n49 = Fresnel(ExponentIn=Constant() {r=0.5095})
  n50 = Multiply(A=Multiply(A=NamedRerouteUsage(), B=@n45.R), B=Constant() {r=0.5})
  n51 = LinearInterpolate(
   A=LinearInterpolate(
    A=LinearInterpolate(A=@n41.RGB, B=@n42.RGB, Alpha=@n44)
    B=VectorParameter() {group=SpecterDunkFaceAcc, parameterName=SkullEmissive_ColorE, defaultValue=<rgba 14 3.8157 0.7 1>}.RGB
    Alpha=Saturate(None=@n48)
   )
   B=VectorParameter() {group=SpecterDunkFaceAcc, parameterName=SkullEmissive_Fresnel, defaultValue=<rgba 1 0.0808 0 1>}.RGB
   Alpha=Saturate(None=Add(A=@n49, B=@n50))
  )
  n52 = Multiply(A=@n51, B=ComponentMask(None=NamedRerouteUsage()) {r=true})
  n53 = Constant() {r=0.5}
  n54 = Constant() {r=0.32}
  n55 = NamedRerouteUsage()
  n56 = Constant() {r=1.5296}
  n57 = Saturate(None=Power(Base=Multiply(A=@n55, B=@n56) {constB=1.3}) {constExponent=5})
  n58 = Multiply(A=Multiply(A=LinearInterpolate(A=@n51, B=@n52, Alpha=@n53), B=LinearInterpolate(A=@n54, Alpha=@n57)), B=NamedRerouteUsage())
  NamedRerouteDeclaration(None=Add(A=Add(A=@n40, B=@n58), B=NamedRerouteUsage())) {name=Out_Emissive, nodeColor=<rgba 1 0.3984 0 1>}
 ]
}
